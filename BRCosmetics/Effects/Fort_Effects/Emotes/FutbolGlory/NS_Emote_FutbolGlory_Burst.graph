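# Niagara one-shot confetti burst for the FutbolGlory emote, using a single sprite emitter.
# Confetti spawns at once in a flattened sphere above the origin, launched along shape normals, with colors picked from a warm HDR array and tinted by time of day.
# Drag, randomized gravity and panning curl noise make it flutter down while it pops in, shrinks and advances through sub-image frames.
# The sphere radius curve has no effect because both lerp endpoints match.
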NiagaraSystem /BRCosmetics/Effects/Fort_Effects/Emotes/FutbolGlory/NS_Emote_FutbolGlory_Burst
dependencies {
 usedDataInterfaces [&NiagaraDataInterfaceArrayColor, &NiagaraDataInterfaceCurve]
 usedDynamicInputs []
 usedModules [&SystemState]
}
emitters {ConfettiBurst {
  emitterUpdateScript [
   &EmitterState()
   &SpawnBurst_Instantaneous(Spawn Count=80, Use Spawn Probability=false, UseLoopCountLimit=true)
  ]
  particleSpawnScript [
   &InitializeParticle(
    Color=&SelectLinearColorFromArray(
     Array Sampling Mode=::NewEnumerator0
     Direct Array Index=?
     Color Selection Array=DataInterface() {propertyValues {ColorData [<rgba 2 1.2806 0.1957 1>, <rgba 5 4.4057 3.6132 1>, <rgba 2 1.3435 0.9978 1>]}}
     Random Seed=0
    )
    Lifetime Max=2.8
    Lifetime Mode=::NewEnumerator1
    Mass Mode=::NewEnumerator3
    Sprite Size Mode=::NewEnumerator1
    Uniform Sprite Size Max=6
    Uniform Sprite Size Min=3
    UsePositionOffset=false
    Write Material Random=true
   )
   &ShapeLocation(
    MaskSpawnGroup=false
    Non Uniform Scale=<1 1 0.5>
    Offset=<0 0 200>
    Offset Mode=::NewEnumerator0
    Override Seed=false
    Sphere Radius=&Lerp_Float(
     A=100
     B=100
     Alpha=&FloatFromCurve(
      FloatCurve=DataInterface() {propertyValues {Curve {keys [
          {interpMode=RCIM_Cubic, tangentMode=RCTM_User, time=0, value=0}
          {arriveTangent=1, interpMode=RCIM_Cubic, leaveTangent=1, time=0.5, value=0.25}
          {arriveTangent=1.9088, interpMode=RCIM_Cubic, leaveTangent=1.9088, tangentMode=RCTM_User, time=1, value=1}
         ]}}}
      CurveIndex=$Emitter.Age
     )
    )
    UseOwnerScale=true
   )
   &Color(Color=&MultiplyLinearColor(
     Linear Color=$Particles.Color
     Scale Factor=&MakeVector4_FromLinearColor(LinearColor=&FNTimeOfDayColor(
       Night Color=<rgba 0.3782 0.3614 0.4948 1>
       Morning Color=<rgba 1 0.7262 0.6963 1>
       Day Color=<rgba 1 0.9315 0.9012 1>
       Evening Color=<rgba 1 0.9069 0.7687 1>
      ))
    ))
   &AddVelocity(
    Velocity=$Particles.ShapeLocation.ShapeNormal
    Velocity Speed Scale=&RandomRangeFloat(Minimum=80, Maximum=160)
   )
  ]
  particleUpdateScript [
   &Drag(Drag=3.5, Use Linear Drag=true, Use Rotational Drag=true)
   &GravityForce(Gravity=&RandomRangeVector(Minimum=<0 0 -110>, Maximum=<0 0 -400>))
   &CurlNoiseForce(
    Bias Noise=false
    Noise Frequency=10
    Noise Strength=&Multiply_Float(A=&RandomRangeFloat(Minimum=30, Maximum=50), B=10)
    Pan Noise=true
    Pan Noise Field=<0.1 0.1 0.1>
   )
   &InheritVelocity(Inherited Velocity Speed Limit=1, Speed Limit=true, UseSpeedThreshold=false)
   &SolveForcesAndVelocity(Clamp Velocity=false, Limit Acceleration=false)
   &ParticleState()
   &ScaleSpriteSize(Uniform Curve Sprite Scale=DataInterface() {propertyValues {Curve {keys [
        {interpMode=RCIM_Cubic, leaveTangent=12.0846, tangentMode=RCTM_Break, time=0, value=0}
        {interpMode=RCIM_Cubic, tangentMode=RCTM_Break, time=0.1, value=1}
        {arriveTangent=-2.7226, tangentMode=RCTM_User, time=1, value=0}
       ]}}})
   /SetVariables(Particles.SubImageIndex=&FloatFromCurve(FloatCurve=DataInterface() {propertyValues {Curve {keys [<at 0 0>,<at 1 1>]}}}, Scale Curve=36)) {isSetParameters=true}
  ]
  rendererClasses [&NiagaraSpriteRendererProperties]
  renderers [{rendererClass=&NiagaraSpriteRendererProperties}]
 }}
summary {
 systemName=NS_Emote_FutbolGlory_Burst
 emitters [{emitterName=ConfettiBurst, rendererClasses [&NiagaraSpriteRendererProperties]}]
 userVariables [
  User.bIsFrontEnd {type=NiagaraBool}
  User.bIsFrontEndPreview {type=NiagaraBool}
  User.NormalizedNotifyProgress {type=NiagaraFloat, defaultValue=0, description=""}
 ]
}
