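# Looping Niagara effect for the FutbolGlory emote, combining falling sprite confetti with an overhead spotlight.
# Confetti spawns in a flattened sphere above the origin, moved by drag, gravity and curl noise.
# Spotlight fades in by emitter age, driving meshes, a light renderer and component light intensity.
# User.CrowdCheerCurve scales confetti rate and turbulence plus spotlight brightness, while time-of-day and front-end flags adjust the look.
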NiagaraSystem /BRCosmetics/Effects/Fort_Effects/Emotes/FutbolGlory/NS_Emote_FutbolGlory_Looping
dependencies {
 usedDataInterfaces [&NiagaraDataInterfaceCurve, &NiagaraDataInterfaceArrayColor]
 usedDynamicInputs []
 usedModules [&SystemState]
}
emitters {
 Confetti {
  emitterUpdateScript [
   &EmitterState(
    Delay First Loop Only=true
    Inactive Response=::NewEnumerator1
    Life Cycle Mode=::NewEnumerator1
    Loop Delay=0.6
    Loop Duration=1
    UseLoopDelay=true
   )
   &SpawnRate(SpawnRate=&Lerp_Float(A=40, B=200, Alpha=$User.CrowdCheerCurve), Use Spawn Probability=false)
  ]
  particleSpawnScript [
   &InitializeParticle(
    Color=&SelectLinearColorFromArray(
     Array Sampling Mode=::NewEnumerator0
     Direct Array Index=?
     Color Selection Array=DataInterface() {propertyValues {ColorData [<rgba 2 1.2806 0.1957 1>, <rgba 5 4.4057 3.6132 1>, <rgba 2 1.3435 0.9978 1>]}}
     Random Seed=0
    )
    Lifetime Max=2.8
    Lifetime Mode=::NewEnumerator1
    Sprite Size Mode=::NewEnumerator1
    Uniform Sprite Size Max=6
    Uniform Sprite Size Min=3
    UsePositionOffset=false
    Write Material Random=true
   )
   &ShapeLocation(
    MaskSpawnGroup=false
    Non Uniform Scale=<1 1 0.5>
    Offset=<0 0 200>
    Offset Mode=::NewEnumerator0
    Override Seed=false
    Sphere Radius=&Lerp_Float(
     A=100
     B=100
     Alpha=&FloatFromCurve(
      FloatCurve=DataInterface() {propertyValues {Curve {keys [
          {interpMode=RCIM_Cubic, tangentMode=RCTM_User, time=0, value=0}
          {arriveTangent=1, interpMode=RCIM_Cubic, leaveTangent=1, time=0.5, value=0.25}
          {arriveTangent=1.9088, interpMode=RCIM_Cubic, leaveTangent=1.9088, tangentMode=RCTM_User, time=1, value=1}
         ]}}}
      CurveIndex=$Emitter.Age
     )
    )
    UseOwnerScale=true
   )
   &Color(Color=&MultiplyLinearColor(
     Linear Color=$Particles.Color
     Scale Factor=&MakeVector4_FromLinearColor(LinearColor=&FNTimeOfDayColor(
       Night Color=<rgba 0.3782 0.3614 0.4948 1>
       Morning Color=<rgba 1 0.7262 0.6963 1>
       Day Color=<rgba 1 0.9315 0.9012 1>
       Evening Color=<rgba 1 0.9069 0.7687 1>
      ))
    ))
   &InheritVelocity(Inherited Velocity Speed Limit=1000, Speed Limit=true, UseSpeedThreshold=false) {module=InheritVelocity001}
  ]
  particleUpdateScript [
   &Drag(Drag=3.5, Use Linear Drag=true, Use Rotational Drag=true)
   &GravityForce(Gravity=&RandomRangeVector(Minimum=<0 0 -110>, Maximum=<0 0 -400>))
   &CurlNoiseForce(
    Bias Noise=false
    Noise Frequency=10
    Noise Strength=&Multiply_Float(
     A=&RandomRangeFloat(Minimum=30, Maximum=50)
     B=&Lerp_Float(A=1, B=10, Alpha=$User.CrowdCheerCurve)
    )
    Pan Noise=true
    Pan Noise Field=<0.1 0.1 0.1>
   )
   &SolveForcesAndVelocity(Clamp Velocity=false, Limit Acceleration=false)
   &ParticleState()
   &ScaleSpriteSize(Uniform Curve Sprite Scale=DataInterface() {propertyValues {Curve {keys [
        {interpMode=RCIM_Cubic, leaveTangent=12.0846, tangentMode=RCTM_Break, time=0, value=0}
        {interpMode=RCIM_Cubic, tangentMode=RCTM_Break, time=0.1, value=1}
        {arriveTangent=-2.7226, tangentMode=RCTM_User, time=1, value=0}
       ]}}})
   /SetVariables(Particles.SubImageIndex=&FloatFromCurve(FloatCurve=DataInterface() {propertyValues {Curve {keys [<at 0 0>,<at 1 1>]}}}, Scale Curve=36)) {isSetParameters=true}
  ]
  rendererClasses [&NiagaraSpriteRendererProperties]
  renderers [{rendererClass=&NiagaraSpriteRendererProperties}]
 }
 Spotlight {
  emitterUpdateScript [
   &EmitterState(
    Inactive Response=::NewEnumerator1
    Life Cycle Mode=::NewEnumerator1
    Loop Behavior=::NewEnumerator1
    Loop Duration Mode=::NewEnumerator1
    UseLoopDelay=false
   )
   &SpawnBurst_Instantaneous(Use Spawn Probability=false, UseLoopCountLimit=false)
  ]
  particleSpawnScript [
   &InitializeParticle(
    Color=&FNTimeOfDayColor(
     Night Color=<rgba 0.5 0.4132 0.25 0.6>
     Morning Color=<rgba 0.5 0.4132 0.25 0.6>
     Day Color=&MakeCustomLinearColorFromBool(
      Bool=&Boolean_OrOperation(Bool1=$User.bIsFrontEnd, Bool2=$User.bIsFrontEndPreview)
      True Color=<rgba 0.8 0.632 0.24 0.275>
      False Color=<rgba 1 0.6877 0.2311 0.5>
     )
     Evening Color=<rgba 0.5 0.4132 0.25 0.6>
    )
    Lifetime Max=0.6
    Lifetime Min=0.8
    Lifetime Mode=::NewEnumerator1
    Mesh Scale=<1.8 1.8 2>
    Mesh Scale Mode=::NewEnumerator4
    Position Offset=<0 0 320>
    Sprite Size Mode=::NewEnumerator3
    Uniform Sprite Size=77
    UsePositionOffset=true
    Write Material Random=true
   )
   &InitialMeshOrientation(Mesh Orientation Mode=::NewEnumerator7, Use Rotation Vector=true)
   /SetVariables(Particles.GlowCameraOffset=33) {isSetParameters=true}
  ]
  particleUpdateScript [
   &ParticleState(Kill Particles When Lifetime Has Elapsed=false, Let Infinitely Lived Particles Die When Emitter Deactivates=true)
   &ScaleColor(
    Scale Alpha=&Multiply_Float(
     A=&Lerp_Float(A=0.45, B=1.1, Alpha=$User.CrowdCheerCurve)
     B=&FloatFromCurve(
      FloatCurve=DataInterface() {propertyValues {Curve {keys [<at 0.6 0 RCIM_Cubic>, {arriveTangent=6.6014, interpMode=RCIM_Cubic, leaveTangent=6.6014, tangentMode=RCTM_User, time=0.95, value=1}]}}}
      CurveIndex=$Emitter.Age
     )
    )
    ScaleA=true
    ScaleRGB=false
   )
   &FN_Light_Attributes(
    Color=$Particles.Color
    Create Color Attribute=true
    Create Exponent Attribute=true
    Create Position Attribute=true
    Create Radius Attribute=true
    Exponent=6
    Position=&AddVectorToPosition(Position=$Particles.Position, Vector=<0 10 -150>)
    Radius=100
    Scale Color=&Multiply_Float(
     A=&Lerp_Float(A=10, B=25, Alpha=$User.CrowdCheerCurve)
     B=&FloatFromCurve(
      FloatCurve=DataInterface() {propertyValues {Curve {keys [<at 0.6 0 RCIM_Cubic>, {arriveTangent=6.6014, interpMode=RCIM_Cubic, leaveTangent=6.6014, tangentMode=RCTM_User, time=0.95, value=1}]}}}
      CurveIndex=$Emitter.Age
     )
    )
   )
   /SetVariables(Particles.Component Light Intensity:NiagaraFloat=&Multiply_Float(
     A=&Multiply_Float(
      A=&Lerp_Float(A=100, B=450, Alpha=$User.CrowdCheerCurve)
      B=&FloatFromCurve(
       FloatCurve=DataInterface() {propertyValues {Curve {keys [
           {interpMode=RCIM_Cubic, tangentMode=RCTM_User, time=0, value=0.6}
           {interpMode=RCIM_Cubic, tangentMode=RCTM_User, time=0.5, value=1}
           {tangentMode=RCTM_User, time=1, value=0.6}
          ]}}}
       CurveIndex=Linked() {linkedVariable {name=NPC.FortniteNPC.TimeOfDay_0-1, type=NiagaraFloat}}
      )
     )
     B=&FloatFromCurve(
      FloatCurve=DataInterface() {propertyValues {Curve {keys [<at 0.6 0 RCIM_Cubic>, {arriveTangent=6.6014, interpMode=RCIM_Cubic, leaveTangent=6.6014, tangentMode=RCTM_User, time=0.95, value=1}]}}}
      CurveIndex=$Emitter.Age
     )
    )) {isSetParameters=true}
  ]
  rendererClasses [
   &NiagaraMeshRendererProperties
   &NiagaraLightRendererProperties
   &NiagaraComponentRendererProperties
  ]
  renderers [
   {rendererClass=&NiagaraMeshRendererProperties, rendererIndex=0}
   {rendererClass=&NiagaraMeshRendererProperties, rendererIndex=1}
   {rendererClass=&NiagaraLightRendererProperties, rendererIndex=2}
   {rendererClass=&NiagaraComponentRendererProperties, rendererIndex=3}
   {rendererClass=&NiagaraMeshRendererProperties, rendererIndex=4}
  ]
 }
}
summary {
 systemName=NS_Emote_FutbolGlory_Looping
 emitters [
  {
   emitterName=Spotlight
   rendererClasses [
    &NiagaraMeshRendererProperties
    &NiagaraLightRendererProperties
    &NiagaraComponentRendererProperties
   ]
  }
  {emitterName=Confetti, rendererClasses [&NiagaraSpriteRendererProperties]}
 ]
 userVariables [
  User.bIsFrontEnd {type=NiagaraBool}
  User.bIsFrontEndPreview {type=NiagaraBool}
  User.CrowdCheerCurve {type=NiagaraFloat, defaultValue=0, description=""}
  User.NormalizedNotifyProgress {type=NiagaraFloat, defaultValue=0, description=""}
 ]
}
